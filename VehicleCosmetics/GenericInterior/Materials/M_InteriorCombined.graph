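# Vehicle interior surface material that combines fixed chair textures with parameterized filler textures, selected by vertex color red.
# Masks apply CustomColor to painted areas and mark fabric, with an optional static-switch diffuse tint.
# Fabric layers distorted microfiber detail, noise-modulated Fresnel rim shading and angle-corrected detail normals, while raising roughness.
# AO is multiplied into base color, and the attributes feed MF_VehicleDeRez_Transition for the de-rez effect.
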Material /VehicleCosmetics/GenericInterior/Materials/M_InteriorCombined
graph {
 properties {
  bUseMaterialAttributes=true
  bUsedWithMeshDeformer=true
  bUsedWithNanite=true
  bUsedWithSkeletalMesh=true
  pixelDepthOffsetMode=PDOM_Legacy
  usedShadingModels=Default Lit
 }
 nodes [
  Constant3Vector() {constant=<rgba 0 0 1 0>}
  n1 = VectorParameter() {parameterName=CustomColor, defaultValue=<rgba 0.055 0.0607 0.0677 1>}
  n2 = TextureSample() {texture=/VehicleCosmetics/GenericInterior/Textures/T_Interior_Chair_BC}
  n3 = StaticSwitchParameter(True=Multiply(A=@n1.RGB, B=@n2.RGB), False=@n2.RGB) {parameterName=TintDiffuse}
  n4 = TextureSample() {samplerType=SAMPLERTYPE_LinearColor, texture=/VehicleCosmetics/GenericInterior/Textures/T_Interior_Chair_M}
  n5 = StaticSwitchParameter(True=@n4.R, False=@n4.R) {parameterName=TintDiffuse}
  n6 = TextureSampleParameter2D() {parameterName=Filler_BaseColor, texture=/VehicleCosmetics/GenericInterior/Textures/T_Interior_Filler_BC}
  n7 = StaticSwitchParameter(True=Multiply(A=@n1.RGB, B=@n6.RGB), False=@n6.RGB) {parameterName=TintDiffuse}
  n8 = TextureSampleParameter2D() {
   parameterName=Filler_Mask
   samplerType=SAMPLERTYPE_LinearColor
   texture=/VehicleCosmetics/Bodies/Spartan/Textures/Interior/T_Spartan_Interior_M
   channelNames {a="", b="", g=Fabric Mask, r=Painted Mask}
  }
  n9 = StaticSwitchParameter(True=@n8.Painted Mask, False=@n8.Painted Mask) {parameterName=TintDiffuse}
  n10 = VertexColor()
  n11 = VectorParameter() {parameterName=Fabric_XForm, defaultValue=<rgba 0 0 12 12>}
  n12 = Multiply(A=Add(A=TextureCoordinate() {coordinateIndex=1}, B=AppendVector(A=@n11.R, B=@n11.R)), B=AppendVector(A=@n11.B, B=@n11.B))
  n13 = VectorParameter() {group=Fiber_Direction_Noise, parameterName=FresnelNoise_XForm, defaultValue=<rgba 0 0 6 6>}
  n14 = Add(
   A=MaterialFunctionCall(
    UVs=TextureCoordinate() {coordinateIndex=1}
    "Rotation Angle (0-1)"=ScalarParameter() {group=Fiber_Direction_Noise, parameterName=Fresnel_Noise_Rotate}
   ) {materialFunction=&CustomRotator}.Rotated Values
   B=AppendVector(A=@n13.R, B=@n13.R)
  )
  n15 = AppendVector(A=@n13.B, B=@n13.B)
  n16 = TextureSample(UVs=@n12) {samplerType=SAMPLERTYPE_LinearGrayscale, texture=/VehicleCosmetics/GenericInterior/Textures/T_Detail_Microfibers_M}
  n17 = ScalarParameter() {defaultValue=0.1, group=Fiber_Direction_Noise, parameterName=Distort_FresnelNoise_w/FabricPattern}
  n18 = TextureSampleParameter2D(UVs=Add(A=Multiply(A=@n14, B=@n15), B=Multiply(A=@n16.G, B=@n17) {constB=0.1})) {
   group=Fiber_Direction_Noise
   parameterName=Fresnel_Noise
   samplerType=SAMPLERTYPE_LinearGrayscale
   texture=/VehicleCosmetics/GenericInterior/Textures/T_Detail_Swipes_M
  }
  n19 = Power(Base=@n18.G, Exp=ScalarParameter() {defaultValue=2, group=Fiber_Direction_Noise, parameterName=Fresnel_Noise_Power})
  n20 = ScalarParameter() {defaultValue=0.033, parameterName=FabricDistortionBlend}
  n21 = LinearInterpolate(A=@n12, B=@n19, Alpha=Clamp(None=Multiply(A=@n19, B=@n20) {constB=0}))
  n22 = Clamp(None=Add(
    A=TextureSample(UVs=@n21) {samplerType=SAMPLERTYPE_LinearGrayscale, texture=&T_Detail_Microfibers_M}.G
    B=ScalarParameter() {parameterName=Fabric_AO_Intensity}
   ))
  n23 = VectorParameter() {parameterName=BaseColor, defaultValue=<rgba 0.0333 0.0333 0.03 0>}
  n24 = VectorParameter() {parameterName=RimColor, defaultValue=<rgba 0.1703 0.1661 0.2031 1>}
  n25 = TextureSample(UVs=@n21) {samplerType=SAMPLERTYPE_Normal, texture=/VehicleCosmetics/GenericInterior/Textures/T_Detail_Microfibers_N}
  n26 = Multiply(
   A=Fresnel(ExponentIn=ScalarParameter() {defaultValue=2, parameterName=Fresnel_Power}, Normal=Transform(None=@n25.RGB)) {baseReflectFraction=0.01, exponent=3}
   B=@n19
  )
  n27 = Multiply(A=@n22, B=@n26)
  n28 = Multiply(A=@n22, B=LinearInterpolate(A=@n23.RGB, B=@n24.RGB, Alpha=Multiply(A=@n27, B=@n26)))
  n29 = Multiply(A=@n28, B=Fresnel() {baseReflectFraction=0.01, exponent=1})
  n30 = LinearInterpolate(
   A=LinearInterpolate(A=LinearInterpolate(A=@n3, B=@n1.RGB, Alpha=@n5), B=LinearInterpolate(A=@n7, B=@n1.RGB, Alpha=@n9), Alpha=@n10.R)
   B=LinearInterpolate(A=Add(A=@n28, B=@n29), B=LinearInterpolate(A=@n24.RGB, B=@n23.RGB, Alpha=@n26), Alpha=@n26)
   Alpha=LinearInterpolate(A=@n4.G, B=@n8.Fabric Mask, Alpha=@n10.R) {desc=LerpFabricMaskByVertexColor}
  ) {desc=LerpBaseColorAndFabricColor}
  n31 = TextureSample() {texture=/VehicleCosmetics/GenericInterior/Textures/T_Interior_Chair_PBR}
  n32 = TextureSampleParameter2D() {
   parameterName=Filler_PBR
   texture=/VehicleCosmetics/GenericInterior/Textures/T_Interior_Filler_PBR
   channelNames {a="", b=AO, g=Roughness, r=Metallic}
  }
  n33 = LinearInterpolate(A=@n31.B, B=@n32.AO, Alpha=@n10.R)
  n34 = Add(A=@n31.G, B=@n4.G)
  n35 = Add(A=@n32.Roughness, B=@n8.Fabric Mask)
  n36 = TextureSample() {samplerType=SAMPLERTYPE_Normal, texture=/VehicleCosmetics/GenericInterior/Textures/T_Interior_Chair_N}
  n37 = Multiply(A=@n25.RGB, B=VectorParameter() {parameterName=Fabric_Normal_Strength, defaultValue=<rgba 0.25 0.25 1 1>}.RGB)
  n38 = LinearInterpolate(
   A=@n36.RGB
   B=MaterialFunctionCall(BaseNormal=@n36.RGB, AdditionalNormal=@n37) {materialFunction=&BlendAngleCorrectedNormals}.Result
   Alpha=@n4.G
  )
  n39 = TextureSampleParameter2D() {parameterName=Filler_Normal, samplerType=SAMPLERTYPE_Normal, texture=/VehicleCosmetics/Textures/T_BlankNormal_B}
  n40 = LinearInterpolate(
   A=@n39.RGB
   B=MaterialFunctionCall(BaseNormal=@n39.RGB, AdditionalNormal=@n37) {materialFunction=&BlendAngleCorrectedNormals}.Result
   Alpha=@n8.Fabric Mask
  )
  MaterialFunctionCall(In=MakeMaterialAttributes!(
    BaseColor=Multiply(A=@n30, B=@n33)
    Metallic=LinearInterpolate(A=@n31.R, B=@n32.Metallic, Alpha=@n10.R)
    Specular
    Roughness=LinearInterpolate(A=@n34, B=@n35, Alpha=@n10.R)
    Anisotropy
    EmissiveColor
    Opacity
    OpacityMask
    Normal=LinearInterpolate(A=@n38, B=@n40, Alpha=@n10.R)
    Tangent
    WorldPositionOffset
    SubsurfaceColor
    ClearCoat
    ClearCoatRoughness
    AmbientOcclusion
    Refraction
    CustomizedUVs_0
    CustomizedUVs_1
    CustomizedUVs_2
    CustomizedUVs_3
    CustomizedUVs_4
    CustomizedUVs_5
    CustomizedUVs_6
    CustomizedUVs_7
    PixelDepthOffset
    ShadingModel
    Displacement
   )) {materialFunction=&MF_VehicleDeRez_Transition}
  VectorParameter() {parameterName=EmissiveColor}
 ]
}
