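# Master material for generic vehicle interiors, defaulting to chair textures, usable on skeletal and Nanite meshes.
# An optional static switch tints the base texture by CustomColor, then the red mask blends toward that flat color and the packed texture's blue channel multiplies the result.
# The blue mask scales a squared EmissiveColor, the green mask is added to the packed green channel, and a normal map is sampled.
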Material /VehicleCosmetics/GenericInterior/Materials/M_Vehicle_Interior
graph {
 properties {
  bUsedWithMeshDeformer=true
  bUsedWithNanite=true
  bUsedWithSkeletalMesh=true
  pixelDepthOffsetMode=PDOM_Legacy
  usedShadingModels=Default Lit
 }
 nodes [
  n1 = TextureSampleParameter2D() {parameterName=InteriorTextures, texture=/VehicleCosmetics/GenericInterior/Textures/T_Interior_Chair_PBR}
  n2 = TextureSampleParameter2D() {parameterName=Masks, samplerType=SAMPLERTYPE_LinearColor, texture=/VehicleCosmetics/GenericInterior/Textures/T_Interior_Chair_M}
  Add(A=@n1.G, B=@n2.G)
  n3 = VectorParameter() {parameterName=EmissiveColor}
  Multiply(A=@n2.B, B=Multiply(A=@n3.RGB, B=@n3.RGB))
  n4 = VectorParameter() {parameterName=CustomColor, defaultValue=<rgba 0.055 0.0607 0.0677 1>}
  n5 = TextureSampleParameter2D() {parameterName=BaseColor, texture=/VehicleCosmetics/GenericInterior/Textures/T_Interior_Chair_BC}
  n6 = Multiply(A=@n4.RGB, B=@n5.RGB)
  Multiply(A=LinearInterpolate(A=StaticSwitchParameter(True=@n6, False=@n5.RGB) {parameterName=TintDiffuse}, B=@n4.RGB, Alpha=@n2.R), B=@n1.B)
  TextureSampleParameter2D() {parameterName=Normal, samplerType=SAMPLERTYPE_Normal, texture=/VehicleCosmetics/GenericInterior/Textures/T_Interior_Chair_N}
 ]
}
